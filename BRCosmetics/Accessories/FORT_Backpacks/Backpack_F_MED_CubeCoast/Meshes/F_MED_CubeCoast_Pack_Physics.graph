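# Physics asset for the CubeCoast backpack cosmetic, defining capsule bodies and joints for its bottle, map and ribbon bow.
# The bottle is the root of the constraint tree, with the map on a tight joint and a zero-size ribbon base on a free-rotating one.
# Two ribbon loops and two three-segment strands branch from that base, with moderate swing and very little twist.
# Strand capsules grow wider and shorter toward the tips.
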PhysicsAsset /BRCosmetics/Accessories/FORT_Backpacks/Backpack_F_MED_CubeCoast/Meshes/F_MED_CubeCoast_Pack_Physics
bodies {
 Bottle_JNT [{shapeType=Capsule, length=8, radius=5.8, shapeName=Bottle_JNT_capsule, center=<-2.666 0.2348 0.5413>, rotation=<rot 0 -90.9098 -87.8342>}]
 Lf_RibbonStrand_01_JNT [{shapeType=Capsule, length=2, radius=0.6, shapeName=Lf_RibbonStrand_01_JNT_capsule, center=<1.7 0 0>, rotation=<rot 90 0 0>}]
 Lf_RibbonStrand_02_JNT [{shapeType=Capsule, length=1.5, radius=1.2, shapeName=Lf_RibbonStrand_02_JNT_capsule, center=<2.3 0 0.35>, rotation=<rot 90 0 0>}]
 Lf_RibbonStrand_03_JNT [{shapeType=Capsule, length=1, radius=1.7, shapeName=Lf_RibbonStrand_03_JNT_capsule, center=<2.7 0 0.75>, rotation=<rot 90 0 0>}]
 Lf_Ribbon_JNT [{shapeType=Capsule, length=0.5, radius=1.8, shapeName=Lf_Ribbon_JNT_capsule, center=<2 0 0>, rotation=<rot 90 0 0>}]
 Map_JNT [{shapeType=Capsule, length=8, radius=2, shapeName=Map_JNT_capsule, center=<6 0 0>, rotation=<rot 90 0 0>}]
 Ribbon_Base_JNT [{shapeType=Capsule, length=0, radius=0, shapeName=Ribbon_Base_JNT_capsule, center=<-0.0436 -0.1428 0.114>, rotation=<rot 0 8.1045 -45.7497>}]
 Rt_RibbonStrand_01_JNT [{shapeType=Capsule, length=2, radius=0.6, shapeName=Rt_RibbonStrand_01_JNT_capsule, center=<1.7 0 0>, rotation=<rot 90 0 0>}]
 Rt_RibbonStrand_02_JNT [{shapeType=Capsule, length=1.5, radius=1.2, shapeName=Rt_RibbonStrand_02_JNT_capsule, center=<2.3 0 0.35>, rotation=<rot 90 0 0>}]
 Rt_RibbonStrand_03_JNT [{shapeType=Capsule, length=1, radius=1.7, shapeName=Rt_RibbonStrand_03_JNT_capsule, center=<2.7 0 0.75>, rotation=<rot 90 0 0>}]
 Rt_Ribbon_JNT [{shapeType=Capsule, length=0.5, radius=1.8, shapeName=Rt_Ribbon_JNT_capsule, center=<2 0 0>, rotation=<rot 90 0 0>}]
}
constraints [
 {
  bone1Name=Ribbon_Base_JNT
  bone2Name=Bottle_JNT
  swing1LimitDegrees=45
  swing1Motion=Free
  swing2LimitDegrees=45
  swing2Motion=Free
  twistLimitDegrees=45
  twistMotion=Free
 }
 {bone1Name=Lf_RibbonStrand_01_JNT, bone2Name=Ribbon_Base_JNT, swing1LimitDegrees=35, swing2LimitDegrees=35, twistLimitDegrees=2}
 {bone1Name=Lf_RibbonStrand_02_JNT, bone2Name=Lf_RibbonStrand_01_JNT, swing1LimitDegrees=35, swing2LimitDegrees=35, twistLimitDegrees=2}
 {bone1Name=Lf_RibbonStrand_03_JNT, bone2Name=Lf_RibbonStrand_02_JNT, swing1LimitDegrees=35, swing2LimitDegrees=35, twistLimitDegrees=2}
 {bone1Name=Rt_RibbonStrand_01_JNT, bone2Name=Ribbon_Base_JNT, swing1LimitDegrees=35, swing2LimitDegrees=35, twistLimitDegrees=2}
 {bone1Name=Rt_RibbonStrand_02_JNT, bone2Name=Rt_RibbonStrand_01_JNT, swing1LimitDegrees=35, swing2LimitDegrees=35, twistLimitDegrees=2}
 {bone1Name=Rt_RibbonStrand_03_JNT, bone2Name=Rt_RibbonStrand_02_JNT, swing1LimitDegrees=35, swing2LimitDegrees=35, twistLimitDegrees=2}
 {bone1Name=Rt_Ribbon_JNT, bone2Name=Ribbon_Base_JNT, swing1LimitDegrees=35, swing2LimitDegrees=35, twistLimitDegrees=2}
 {bone1Name=Lf_Ribbon_JNT, bone2Name=Ribbon_Base_JNT, swing1LimitDegrees=35, swing2LimitDegrees=35, twistLimitDegrees=2}
 {bone1Name=Map_JNT, bone2Name=Bottle_JNT, swing1LimitDegrees=10, swing2LimitDegrees=10, twistLimitDegrees=2}
]
