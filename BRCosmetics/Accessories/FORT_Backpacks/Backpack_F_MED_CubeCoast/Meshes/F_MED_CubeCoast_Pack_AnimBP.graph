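# Animation Blueprint for the CubeCoast backpack cosmetic, adding secondary motion to its starfish, ribbon and map bones.
# A root offset feeds three faint springs, a RigidBody simulation relative to the pack bone, and an interpolated Map_JNT offset during non-skydiving jumps or falls.
# The update event switches springs off above a pawn speed cutoff, also disables the map spring while airborne, and sets RigidBody gravity through a deprecated override function.
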AnimBlueprint /BRCosmetics/Accessories/FORT_Backpacks/Backpack_F_MED_CubeCoast/Meshes/F_MED_CubeCoast_Pack_AnimBP
target_skeleton=&/F_MED_CubeCoast_Pack_Skeleton
graphs [
 AnimGraph {nodes [
   n1 = "Transform (Modify) Bone - Bone: root"(ComponentPose=Local To Component(LocalPose=Input Pose()), Translation="0,1.5,0", Rotation="0,0,0") {
    translationMode=BMM_Additive
    translationSpace=BCS_BoneSpace
    alphaBoolBlend {blendOption=Linear}
    boneToModify {boneName=root}
    translation {y=1.5}
   }
   n2 = Spring controller - Bone: StarFish_Jiggle_JNT!(ComponentPose:Component Space Pose Link Structure=@n1, Alpha:float=Get Spring Alpha() -> (Spring Alpha:double)) -> (Pose:Component Space Pose Link Structure) {
    is=title AnimGraphNode_SpringBone
    springDamping=2.5
    springStiffness=250
    alphaBoolBlend {blendOption=Linear}
    springBone {boneName=StarFish_Jiggle_JNT}
   }
   n3 = Spring controller - Bone: Ribbon_Jiggle_JNT!(ComponentPose:Component Space Pose Link Structure=@n2, Alpha:float=Get Spring Alpha 2() -> (Spring Alpha 2:double)) -> (Pose:Component Space Pose Link Structure) {
    is=title AnimGraphNode_SpringBone
    springDamping=1.5
    springStiffness=150
    alphaBoolBlend {blendOption=Linear}
    springBone {boneName=Ribbon_Jiggle_JNT}
   }
   n4 = RigidBody!(
    ComponentPose=Spring controller - Bone: Map_JNT!(ComponentPose:Component Space Pose Link Structure=@n3, Alpha:float=Get Spring Alpha 3() -> (Spring Alpha 3:double)) -> (Pose:Component Space Pose Link Structure) {
     is=title AnimGraphNode_SpringBone
     springDamping=2.5
     springStiffness=350
     alphaBoolBlend {blendOption=Linear}
     springBone {boneName=Map_JNT}
    }
    Alpha
    OverrideWorldGravity:Vector=Get Gravity Override()
    ExternalForce
   ) -> (Pose) {
    bClampLinearTranslationLimitToRefPose=true
    bOverrideWorldGravity=true
    overridePhysicsAsset=&/F_MED_CubeCoast_Pack_Physics
    simulationSpace=BaseBoneSpace
    alphaBoolBlend {blendOption=Linear}
    baseBoneRef {boneName=pack}
    componentAppliedLinearAccClamp=<3000 3000 3000>
    componentLinearAccScale=<0.4 0.3 0.3>
    componentLinearVelScale=<0.4 0.3 0.3>
    simSpaceSettings {
     maxAngularAcceleration=1000
     maxAngularVelocity=3
     maxLinearAcceleration=700
     maxLinearVelocity=500
     velocityScaleZ=0
     worldAlpha=0.5
     externalLinearDragV=<0.2 0.2>
    }
   }
   n5 = "Transform (Modify) Bone - Bone: Map_JNT"(
    ComponentPose=@n4
    Alpha=Get Is Jumping Falling Alpha() -> (Is Jumping Falling Alpha:float)
    Translation="2.5,0,2"
    Rotation="0,0,0"
    Scale="1,1,1"
   ) {
    translationMode=BMM_Additive
    translationSpace=BCS_BoneSpace
    alphaBoolBlend {blendOption=Linear}
    alphaScaleBiasClamp {bInterpResult=true, interpSpeedDecreasing=20, interpSpeedIncreasing=1.5}
    boneToModify {boneName=Map_JNT}
    translation {x=2.5, z=2}
   }
   Output Pose(Result=Component To Local(ComponentPose=@n5)) {layerGroup=None}
  ]}
 EventGraph {nodes [
   Event BlueprintUpdateAnimation()
   > n1 = Sequence!(execute) -> (then_0, then_1, then_2:Exec)
   > n2 = Set Pawn Velocity Speed(Pawn Velocity Speed:double=Vector Length(A=Get PawnVelocity())) -> (Output_Get:double)
   > n3 = Set Pawn Velocity Speed Cutoff(Pawn Velocity Speed Cutoff:double=SelectFloat(A=0, B=1, bPickA="float > float"(A=@n2, B=1200) {is=title K2Node_CallFunction})) -> (Output_Get:double)
   n4 = float * float(A=0.01, B=@n3) {is=title K2Node_PromotableOperator}
   > n5 = Set Spring Alpha(Spring Alpha:double=@n4) -> (Output_Get:double)
   > Set Spring Alpha 2(Spring Alpha 2:double=float * float(A=0.15, B=@n5)) -> (Output_Get:double)
   n6 = NOT Boolean(A=OR Boolean(A=Get bIsJumping(), B=Get bIsFalling()))
   > Set Spring Alpha 3(Spring Alpha 3:double=float * float!(A=2, B=@n4, C:double="To Float (Boolean)"(InBool=@n6)) -> (ReturnValue)) -> (Output_Get:double)
   n7 = Gravity Override State Function DEPRECATED(
    execute=@n1.then_1
    CharacterPartAnimInstance=Self-Reference()
    GameTimeInSeconds=GetGameTimeInSeconds()
    GravityOverrideParams=Get Gravity Override Params()
   )
   > Set Gravity Override(Gravity Override=@n7)
   n8 = NOT Boolean(A=Get bIsSkyDiving())
   n9 = OR Boolean(A=Get bIsJumping(), B=Get bIsFalling())
   Set Is Jumping Falling Alpha(execute=@n1.then_2, Is Jumping Falling Alpha:float="To Float (Boolean)"(InBool=AND Boolean(A=@n9, B=@n8))) -> (Output_Get:float)
   TryGetPawnOwner()
  ]}
]
variables [
 Gravity Override {type=Vector}
 Gravity Override Params {type=GravityOverrideParamsStruct}
 Is Jumping Falling Alpha {type=number}
 Pawn Velocity Speed {type=number}
 Pawn Velocity Speed Cutoff {type=number}
 Spring Alpha {type=number}
 Spring Alpha 2 {type=number}
 Spring Alpha 3 {type=number}
]
